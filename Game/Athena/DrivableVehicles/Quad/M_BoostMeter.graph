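# Masked material for the Quad vehicle's boost meter, drawing a fill gauge driven by the BoostAmount parameter.
# Static switches pick a radial ring via VectorToRadialValue or a horizontal/vertical UV bar shaped by sphere masks.
# A floored gradient sum cuts the fill, which blends between charging colors and switches to the ready color at full.
# A time-driven sine blinks brightness below BlinkValue, and a switchable opacity mask trims the shape.
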Material /Game/Athena/DrivableVehicles/Quad/M_BoostMeter
graph {
 properties {
  bUsedWithMeshDeformer=true
  bUsedWithSkeletalMesh=true
  blendMode=BLEND_Masked
  pixelDepthOffsetMode=PDOM_Legacy
  usedShadingModels=Default Lit
 }
 nodes [
  n1 = ScalarParameter() {defaultValue=1, parameterName=BoostAmount}
  n2 = Constant() {r=1}
  n3 = Multiply(A=ScalarParameter() {defaultValue=5, parameterName=BlinkSpeed}, B=Time())
  n4 = ScalarParameter() {defaultValue=1.5, parameterName=s}
  n5 = ScalarParameter() {defaultValue=0.5, parameterName=s_1}
  n6 = LinearInterpolate(
   A=VectorParameter() {parameterName=ChargingColor1, defaultValue=<rgba 0.5 0 0 1>}.RGB
   B=VectorParameter() {parameterName=ChargingColor2, defaultValue=<rgba 0.755 0.282 0 1>}.RGB
   Alpha=Power(Base=@n1, Exp=ScalarParameter() {defaultValue=3, parameterName=lerpPow})
  )
  n7 = VectorParameter() {parameterName=ReadyColor, defaultValue=<rgba 0.0618 0.4635 0 1>}
  n8 = Floor(None=@n1)
  n9 = MaterialFunctionCall() {materialFunction=&VectorToRadialValue}
  n10 = OneMinus(None=ComponentMask(None=@n9.Radial Coordinates) {r=true})
  n11 = TextureCoordinate()
  n12 = ComponentMask(None=OneMinus(None=@n11)) {r=true}
  n13 = ComponentMask(None=@n11) {g=true}
  n14 = StaticSwitchParameter(True=@n12, False=@n13) {parameterName=IsHorizontalBar?}
  n15 = OneMinus(None=@n14)
  n16 = Floor(None=Add(A=@n1, B=StaticSwitchParameter(True=@n10, False=@n15) {defaultValue=true, parameterName=IsCircular?}))
  n17 = Subtract(
   A=Distance(A=ComponentMask(None=@n9.Radial Coordinates) {g=true}, B=ScalarParameter() {defaultValue=0.8, parameterName=RadiusDistance})
   B=ScalarParameter() {defaultValue=0.2, parameterName=RadiusSubtract}
  )
  n18 = Multiply(A=@n17) {constB=-1}
  n19 = Constant() {r=0.5}
  n20 = Multiply(
   A=SphereMask(
    A=@n14
    B=@n19
    Radius=ScalarParameter() {defaultValue=0.3, parameterName=XRadius}
    Hardness=ScalarParameter() {defaultValue=0.5, parameterName=XHardness}
   ) {attenuationRadius=0.35}
   B=SphereMask(
    A=StaticSwitchParameter(True=@n13, False=@n12) {parameterName=IsHorizontalBar?}
    B=@n19
    Radius=ScalarParameter() {defaultValue=0.45, parameterName=YRadius}
    Hardness=ScalarParameter() {defaultValue=0.75, parameterName=YHardness}
   ) {attenuationRadius=0.5}
  )
  n21 = StaticSwitchParameter(True=Multiply(A=Saturate(None=@n18), B=ScalarParameter() {defaultValue=10.2575, parameterName=RadiusMultiply}), False=@n20) {defaultValue=true, parameterName=IsCircular?}
  Multiply(
   A=If(
    A=@n1
    B=ScalarParameter() {defaultValue=0.5, parameterName=BlinkValue}
    "A > B"=@n2
    "A == B"=@n2
    "A < B"=MaterialFunctionCall!(Sine Phase=@n3, Value 1=@n4, Value 2=@n5) {materialFunction=/Engine/Functions/Engine_MaterialFunctions02/Utility/Sine_Remapped}.Remapped Sine
   )
   B=Multiply(A=LinearInterpolate(A=@n6, B=@n7.RGB, Alpha=@n8), B=Multiply(A=@n16, B=@n21))
  )
  n22 = StaticSwitchParameter(True=OneMinus(None=Ceil(None=@n17)), False=@n20) {defaultValue=true, parameterName=IsCircular?}
  StaticSwitchParameter(True=Saturate(None=Multiply(A=@n16, B=@n22)), False=ScalarParameter() {defaultValue=1, parameterName=OpacityMaskValue}) {defaultValue=true, parameterName=UseOPacityMask?}
 ]
}
