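# Material for the Quad vehicle's boost meter bar, showing BoostAmount as a fill that grows outward from the center.
# A floored SphereMask sets the fill, colored by a powered charging gradient, CanBoost's boosting color or ReadyColor when full, and pulsed below BlinkValue.
# While CanBoost is on, mirrored panning sine stripes animate the unfilled area.
# A vertical falloff and Brightness scale the result.
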Material /Game/Athena/DrivableVehicles/Quad/M_BoostMeter_Bar
graph {
 properties {bUsedWithMeshDeformer=true, bUsedWithSkeletalMesh=true, pixelDepthOffsetMode=PDOM_Legacy, usedShadingModels=Default Lit}
 nodes [
  MaterialFunctionCall() {materialFunction=&DebugTimeSine}
  n1 = ScalarParameter() {parameterName=BoostAmount}
  n2 = Constant() {r=1}
  n3 = ScalarParameter() {defaultValue=5, parameterName=BlinkSpeed}
  n4 = Time()
  n5 = If(
   A=@n1
   B=ScalarParameter() {defaultValue=0.5, parameterName=BlinkValue}
   "A > B"=@n2
   "A == B"=@n2
   "A < B"=MaterialFunctionCall!(
    Sine Phase=Multiply(A=@n3, B=@n4)
    Value 1=ScalarParameter() {defaultValue=1.5, parameterName=s}
    Value 2=ScalarParameter() {defaultValue=0.5, parameterName=s_1}
   ) {materialFunction=/Engine/Functions/Engine_MaterialFunctions02/Utility/Sine_Remapped}.Remapped Sine
  )
  n6 = LinearInterpolate(
   A=VectorParameter() {parameterName=ChargingColor1, defaultValue=<rgba 0.5 0 0 1>}.RGB
   B=VectorParameter() {parameterName=ChargingColor2, defaultValue=<rgba 0.755 0.282 0 1>}.RGB
   Alpha=Power(Base=@n1, Exp=ScalarParameter() {defaultValue=3, parameterName=lerpPow})
  )
  n7 = VectorParameter() {parameterName=BoostingColor2, defaultValue=<rgba 0.1255 2.4289 3 1>}
  n8 = ScalarParameter() {parameterName=CanBoost}
  n9 = TextureCoordinate()
  n10 = SphereMask(A=ComponentMask(None=@n9) {r=true}, B=Constant() {r=0.5}) {attenuationRadius=0.5, hardnessPercent=0}
  n11 = Floor(None=Add(A=@n1, B=@n10))
  n12 = Multiply(
   A=LinearInterpolate(
    A=LinearInterpolate(A=@n6, B=@n7.RGB, Alpha=@n8)
    B=VectorParameter() {parameterName=ReadyColor, defaultValue=<rgba 0.0618 0.4635 0 1>}.RGB
    Alpha=Floor(None=@n1)
   )
   B=@n11
  )
  n13 = ComponentMask(None=TextureCoordinate()) {r=true}
  n14 = LinearInterpolate(Alpha=Round(None=@n13)) {constA=1, constB=-1}
  n15 = Time()
  n16 = ScalarParameter() {defaultValue=1, parameterName=PanSpeed}
  n17 = Add(A=@n13, B=Multiply(A=@n14, B=Multiply(A=@n15, B=@n16)))
  n18 = ConstantBiasScale(None=Multiply(A=Sine(None=@n17) {period=0.07}, B=@n14))
  n19 = Multiply(A=Multiply(A=Multiply(A=@n18, B=@n10), B=VectorParameter() {parameterName=BoostingColor1, defaultValue=<rgba 0.0425 0.2818 1 1>}.RGB), B=@n8)
  n20 = OneMinus(None=@n11)
  n21 = ComponentMask(None=@n9) {g=true}
  n22 = ScalarParameter() {defaultValue=0.5}
  n23 = Multiply(A=Subtract(A=Distance(A=@n21, B=@n22), B=ScalarParameter() {defaultValue=0.5})) {constB=-1}
  Multiply(
   A=Add(A=Multiply(A=@n5, B=@n12), B=Multiply(A=@n19, B=@n20))
   B=Multiply(A=Saturate(None=@n23), B=ScalarParameter() {defaultValue=10, parameterName=Brightness})
  )
 ]
}
